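# Material function that gathers the customization looks for Creative primitive props into one Material Attributes pass.
# It threads the input attributes serially through nineteen per-style MF_CM functions, from CarbonFiber through effects like GalacticSwirl and Slurpy to surfaces like Marble and Dirt.
# Each function's Result feeds the next, so any style selection happens outside this graph.
# MF_CM_HexPulse branches off after DiscoLights but is unconnected; HexPulse2 takes its place in the chain.
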MaterialFunction /Game/Creative/Sets/PropSets/Primitives/Customization/MaterialFunctions/MF_CM_Primitive_Customization
graph {nodes [
  n1 = MaterialFunctionCall!(In=FunctionInput() {inputType=FunctionInput_MaterialAttributes}) {materialFunction=&/MF_CM_CarbonFiber}
  n2 = MaterialFunctionCall!(In=MaterialFunctionCall!(In=MaterialFunctionCall!(In=@n1.Result) {materialFunction=&/MF_CM_SynthWaveAudio}.Result) {materialFunction=&/MF_CM_GalacticSwirl}.Result) {materialFunction=&/MF_CM_GasCloudRainbow_01}
  n3 = MaterialFunctionCall!(In=MaterialFunctionCall!(In=MaterialFunctionCall!(In=@n2.Result) {materialFunction=&/MF_CM_NeonGasCloud}.Result) {materialFunction=&/MF_CM_Cubed}.Result) {materialFunction=&/MF_CM_DiscoLights}
  n4 = MaterialFunctionCall!(In=@n3.Result) {materialFunction=&/MF_CM_HexPulse2}
  n5 = MaterialFunctionCall!(In=MaterialFunctionCall!(In=MaterialFunctionCall!(In=@n4.Result) {materialFunction=&/MF_CM_Slurpy}.Result) {materialFunction=&/MF_CM_NeuralCosmos}.Result) {materialFunction=&/MF_CM_Cakey}
  n6 = MaterialFunctionCall!(In=MaterialFunctionCall!(In=MaterialFunctionCall!(In=@n5.Result) {materialFunction=&/MF_CM_RainbowStripe}.Result) {materialFunction=&/MF_CM_LavaRocks}.Result) {materialFunction=&/MF_CM_Ice}
  n7 = MaterialFunctionCall!(In=MaterialFunctionCall!(In=MaterialFunctionCall!(In=@n6.Result) {materialFunction=&/MF_CM_Marble}.Result) {materialFunction=&/MF_CM_Brick_Wood_Concrete}.Result) {materialFunction=&/MF_CM_Grass}
  FunctionOutput(None=MaterialFunctionCall!(In=MaterialFunctionCall!(In=@n7.Result) {materialFunction=&/MF_CM_SandCastle}.Result) {materialFunction=&/MF_CM_Dirt}.Result)
  MaterialFunctionCall!(In=@n3.Result) {materialFunction=&/MF_CM_HexPulse}
 ]}
