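# World-grid surface material for Creative primitive props, drawing glowing colored lines over a dark background.
# World position is tiled, raised to a power and combined by vector length into lines that drive base color, specular and emissive.
# A scrolling world-aligned texture fades the lines and varies their sharpness on upward-facing surfaces.
# Attributes pass through primitive customization and game-wide shared functions; both Triplanar switch branches are identical.
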Material /Game/Creative/Sets/PropSets/Primitives/Rounds/Materials/M_Primitive_WorldGrid
graph {
 properties {bUseMaterialAttributes=true, pixelDepthOffsetMode=PDOM_Legacy, usedShadingModels=Default Lit}
 nodes [
  Constant()
  Constant() {r=0.2}
  Constant() {r=0.6}
  n1 = VectorParameter() {group=WorldGrid, parameterName=Grid Color, defaultValue=<rgba 2.677 0.1551 5 0>}
  n2 = WorldPosition()
  n3 = Multiply(A=Constant() {r=100}, B=ScalarParameter() {defaultValue=1, group=WorldGrid, parameterName=Grid Size})
  n4 = Subtract(A=Frac(None=Divide(A=@n2.XYZ, B=@n3) {constB=100})) {constB=0.5}
  n5 = ScalarParameter() {defaultValue=30, group=WorldGrid, parameterName=Line Exp Low}
  n6 = ScalarParameter() {defaultValue=300, group=WorldGrid, parameterName=Line Exp High}
  n7 = Add(A=@n5, B=@n6)
  n8 = ScalarParameter() {defaultValue=1024, group=WorldGrid, parameterName=Panner Texture Size}
  n9 = Multiply(
   A=Time() {bOverride_Period=true, period=3600}
   B=Multiply(A=ScalarParameter() {defaultValue=0.1, group=WorldGrid, parameterName=Scroll Speed}, B=@n8)
  )
  n10 = ComponentMask(None=MaterialFunctionCall(
    TextureObject=TextureObject() {texture=&T_Smooth_tile}
    TextureSize=@n8
    WorldPosition=Add(A=@n2.XYZ, B=@n9)
   ) {materialFunction=/Game/Weapons/WeaponSkins/Wraps/MaterialFunctions/WorldAlignedTexture}.XYZ Texture) {r=true}
  n11 = Saturate(None=DotProduct(A=VertexNormalWS(), B=Constant3Vector() {constant=<rgba 0 0 1 0>}))
  n12 = Power(
   Base=Multiply(A=Abs(None=@n4)) {constB=2}
   Exp=LinearInterpolate(A=Divide(A=@n7), B=LinearInterpolate(A=@n5, B=@n6, Alpha=@n10), Alpha=Multiply(A=@n11, B=@n11))
  ) {constExponent=20}
  n13 = MaterialFunctionCall(Vector 3=@n12, Vector 2=ComponentMask(None=@n12) {g=true, r=true}) {materialFunction=/Engine/Functions/Engine_MaterialFunctions02/Utility/VectorLength}
  n14 = Multiply(A=StaticSwitchParameter(True=@n13.V3 Length, False=@n13.V3 Length) {group=WorldGrid, parameterName=Triplanar}, B=OneMinus(None=@n10))
  n15 = LinearInterpolate(A=VectorParameter() {group=WorldGrid, parameterName=Background Color, defaultValue=<rgba 0.0029 0.0007 0.0052 1>}.RGB, B=@n1.RGB, Alpha=@n14)
  n16 = ScalarParameter() {defaultValue=0.2, group=WorldGrid, parameterName=Metallic}
  n17 = LinearInterpolate(A=ScalarParameter() {defaultValue=0.5, group=WorldGrid, parameterName=Specular}, B=Constant(), Alpha=@n14)
  n18 = ScalarParameter() {defaultValue=0.2, group=WorldGrid, parameterName=Roughness}
  n19 = Multiply(A=@n1.RGB, B=@n14)
  MaterialFunctionCall!(In=MaterialFunctionCall!(In=MakeMaterialAttributes!(
     BaseColor=@n15
     Metallic=@n16
     Specular=@n17
     Roughness=@n18
     Anisotropy
     EmissiveColor=@n19
     Opacity
     OpacityMask
     Normal
     Tangent
     WorldPositionOffset
     SubsurfaceColor
     ClearCoat
     ClearCoatRoughness
     AmbientOcclusion
     Refraction
     CustomizedUVs_0
     CustomizedUVs_1
     CustomizedUVs_2
     CustomizedUVs_3
     CustomizedUVs_4
     CustomizedUVs_5
     CustomizedUVs_6
     CustomizedUVs_7
     PixelDepthOffset
     ShadingModel
     Displacement
    )) {materialFunction=/Game/Creative/Sets/PropSets/Primitives/Customization/MaterialFunctions/MF_CM_Primitive_Customization}.Result) {materialFunction=/Game/Athena/WinterArt/Materials/MF_GameWideShared}
  TextureObject() {texture=&AtmosphericCloudNoise01}
 ]
}
